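# Physics asset for the burger mesh variant of the one-handed Mascot Militia Cuddle pickaxe, a cosmetic melee weapon.
# It defines a single body on the Melee_TwoHanded_Handle bone, using one rotated, offset capsule as the collision shape.
# It has no constraints, so no jointed parts are simulated.
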PhysicsAsset /BRCosmetics/Weapons/FORT_Melee/Pickaxe_MascoMilitia_Cuddle_1H/Meshes/Mascot_Militia_Cuddle_1H_Burger_PhysicsAsset
bodies {Melee_TwoHanded_Handle [{shapeType=Capsule, length=33.6674, radius=23.864, center=<0.0496 6.724 15.028>, rotation=<rot -12.6982 -90 -0.2422>}]}
constraints []
