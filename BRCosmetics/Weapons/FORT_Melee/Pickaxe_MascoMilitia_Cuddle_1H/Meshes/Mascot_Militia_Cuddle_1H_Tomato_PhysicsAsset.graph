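# Physics asset for the Tomato variant of the one-handed Mascot Militia Cuddle pickaxe mesh, a melee weapon cosmetic.
# It has a single capsule body on the Melee_TwoHanded_Handle bone, offset and rotated from the bone origin.
# With no constraints, the whole mesh is represented by that one rigid collision shape, with no jointed parts.
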PhysicsAsset /BRCosmetics/Weapons/FORT_Melee/Pickaxe_MascoMilitia_Cuddle_1H/Meshes/Mascot_Militia_Cuddle_1H_Tomato_PhysicsAsset
bodies {Melee_TwoHanded_Handle [{shapeType=Capsule, length=33.5574, radius=24.0144, center=<0.0578 5.7619 14.0529>, rotation=<rot -9.3176 -90 -0.2693>}]}
constraints []
